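# Unlit translucent mesh-particle material for a spooky mist camera lens effect, drawn without depth testing.
# Triplanar camera-vector noise distorts a panning swirl mask, and a second swirl modulates the contrast power.
# The result blends dark and particle-tinted light colors under a vertical UV falloff, with optional time-of-day curve tinting, and drives alpha.
# Dynamic parameters control panning, power and a CloseUp offset that pulls vertices toward the object position by UV.
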Material /Game/Effects/Fort_Effects/Materials/LensEffect/M_CameraLens_SpookyMist
graph {
 properties {
  bDisableDepthTest=true
  bUsedWithMeshParticles=true
  bUsedWithNiagaraMeshParticles=true
  blendMode=BLEND_Translucent
  pixelDepthOffsetMode=PDOM_Legacy
  shadingModel=MSM_Unlit
  usedShadingModels=Unlit
 }
 nodes [
  n1 = TextureCoordinate()
  n2 = ObjectPositionWS()
  n3 = WorldPosition()
  LinearInterpolate(
   B=Multiply(A=ComponentMask(None=@n1) {g=true}, B=Subtract(A=@n2, B=@n3.XYZ))
   Alpha=DynamicParameter() {defaultValue=<rgba 0 1 1 1>, paramNames [CloseUp,PowerMult,PanSpeed,PanAdd]}.CloseUp
  ) {constAlpha=1}
  MaterialFunctionCall() {materialFunction=&DebugTimeSine}
  n4 = CurveAtlasRowParameter(CurveTime=CollectionParameter() {collection=&FortniteMaterialParameters, parameterName=TimeOfDay_0-1}) {
   atlas=&AthenaCurveAtlas01
   curve=/Game/Athena/LinearColorCurves/Curve_TimeOfDay_S23
   defaultValue=321
   group=TimeOfDay
   parameterName=TOD_Curve
  }
  n5 = ParticleColor()
  n6 = DynamicParameter() {defaultValue=<rgba 0 1 1 1>, paramNames [CloseUp,PowerMult,PanSpeed,PanAdd]}
  n7 = Time()
  n8 = Add(A=@n6.PanAdd, B=Multiply(A=Multiply(A=@n6.PanSpeed, B=@n7), B=ScalarParameter() {defaultValue=3, parameterName=PanSpeed}))
  n9 = Constant3Vector() {constant=<rgba 0 0 0.15 0>}
  n10 = ComponentMask(None=MaterialFunctionCall!(
    CameraVector
    Texture=TextureObject() {texture=&T_Smooth_tile}
    Tiling=ScalarParameter() {defaultValue=2, parameterName=CameraTextureTIle}
    Offset=Multiply(A=@n8, B=@n9)
    AxisFadeContrast=Constant() {r=1}
   ) {materialFunction=/Engine/Functions/Engine_MaterialFunctions01/Texturing/TriplanarCameraVector}.Result) {g=true}
  n11 = ScalarParameter() {defaultValue=1.5, parameterName=TexturePower}
  n12 = AppendVector(A=ScalarParameter() {defaultValue=1, parameterName=XTile}, B=ScalarParameter() {defaultValue=1, parameterName=YTile})
  n13 = TextureCoordinate() {uTiling=3, vTiling=0.2}
  n14 = Constant2Vector() {g=0.008, r=0.025}
  n15 = Panner(Coordinate=Add(A=Multiply(A=@n12, B=@n13), B=Multiply(A=@n14, B=@n10) {constB=0.03}), Time=@n8) {speedX=-0.01, speedY=0.035}
  n16 = ScalarParameter() {defaultValue=7, parameterName=Power}
  n17 = Panner(Coordinate=Multiply(A=TextureCoordinate() {uTiling=2, vTiling=0.1}, B=@n12), Time=@n8) {speedX=0.015, speedY=-0.02}
  n18 = Power(
   Base=LinearInterpolate(
    A=Multiply(A=@n10, B=@n11) {constB=1.5}
    Alpha=TextureSampleParameter2D(UVs=@n15) {parameterName=Mask, samplerType=SAMPLERTYPE_LinearGrayscale, texture=/Game/Global/Textures/Default/Flowmaps/T_StormSwirlMask}.G
   )
   Exp=LinearInterpolate(
    A=Multiply(A=@n16, B=@n6.PowerMult)
    Alpha=TextureSampleParameter2D(UVs=@n17) {parameterName=Mask, samplerType=SAMPLERTYPE_LinearGrayscale, texture=&T_StormSwirlMask}.G
   ) {constA=7, constB=0.6}
  )
  n19 = TextureCoordinate()
  n20 = SphereMask(A=OneMinus(None=ComponentMask(None=@n19) {g=true}), B=Constant() {r=0.5}) {attenuationRadius=0.5, hardnessPercent=0}
  n21 = Power(Base=@n20)
  n22 = LinearInterpolate(
   A=VectorParameter() {parameterName=DarkColor, defaultValue=<rgba 0.0055 0.0075 0.01 0>}.RGB
   B=Multiply(A=@n5.RGB, B=VectorParameter() {parameterName=LightColor, defaultValue=<rgba 0.233 0.1454 0.6 0>}.RGB)
   Alpha=Saturate(None=Multiply(A=@n18, B=@n21))
  )
  MaterialFunctionCall(Emissive=StaticSwitchParameter(True=Multiply(A=@n4, B=@n22), False=@n22) {parameterName=Use ToD Curve}) {materialFunction=&MF_FX_EmissiveAdjustments}
  Multiply(A=Multiply(A=@n5.A, B=@n18), B=@n20)
 ]
}
