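# Unlit modulate-blend material for mesh particles, including Niagara, that tints the scene behind it with a vertical gradient, filed under lens effects.
# Particle alpha fades from white, which leaves the scene unchanged under modulate, to the Color parameter (default purple).
# LinearGradient's VGradient, shaped by a Contrast power, then blends that tint back to white along V.
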Material /Game/Effects/Fort_Effects/Materials/LensEffect/M_Lens_Gadient_Modulate
graph {
 properties {
  bUsedWithMeshParticles=true
  bUsedWithNiagaraMeshParticles=true
  blendMode=BLEND_Modulate
  pixelDepthOffsetMode=PDOM_Legacy
  shadingModel=MSM_Unlit
  translucencyPass=MTP_BeforeDOF
  usedShadingModels=Unlit
 }
 nodes [
  n1 = Constant3Vector() {constant=<rgba 1 1 1 0>}
  LinearInterpolate(
   A=LinearInterpolate(A=@n1, B=VectorParameter() {parameterName=Color, defaultValue=<rgba 0.2231 0.0815 0.401 0>}.RGB, Alpha=ParticleColor().A)
   B=@n1
   Alpha=Power(
    Base=MaterialFunctionCall() {materialFunction=&LinearGradient}.VGradient
    Exp=ScalarParameter() {defaultValue=1.5, parameterName=Contrast}
   )
  )
 ]
}
