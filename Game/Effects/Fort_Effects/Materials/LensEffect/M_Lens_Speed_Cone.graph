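# Translucent unlit mesh-particle material for a speed cone lens effect, drawing streaks that scroll along V.
# A panning cylinder gradient, sharpened and boosted, is added to two tiled, panning noise samples to form the streak pattern.
# Particle color tints it through an emissive-adjustment function, while particle alpha and a shaped V-gradient mask drive opacity.
# A shading path switch applies depth fade by default and a scalar opacity multiplier on mobile.
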Material /Game/Effects/Fort_Effects/Materials/LensEffect/M_Lens_Speed_Cone
graph {
 properties {
  bUsedWithMeshParticles=true
  bUsedWithNiagaraMeshParticles=true
  blendMode=BLEND_Translucent
  pixelDepthOffsetMode=PDOM_Legacy
  shadingModel=MSM_Unlit
  usedShadingModels=Unlit
 }
 nodes [
  n1 = ParticleColor()
  n2 = TextureSample(UVs=Panner(Coordinate=TextureCoordinate()) {speedY=-2}) {samplerSource=SSM_Wrap_WorldGroupSettings, texture=/Game/Global/Textures/Default/Masks_and_Gradients/CylGradient}
  n3 = Panner(Coordinate=TextureCoordinate() {uTiling=5}) {speedY=-2}
  n4 = Panner(Coordinate=TextureCoordinate() {uTiling=10, vTiling=0.25}) {speedY=-2}
  n5 = Add(
   A=Multiply(A=Power(Base=@n2.G) {constExponent=4}) {constB=60}
   B=Add(
    A=TextureSample(UVs=@n3) {texture=&T_NoisePack_01}.R
    B=TextureSample(UVs=@n4) {texture=&T_NoisePack_01}.G
   )
  )
  MaterialFunctionCall(Emissive=Multiply(A=@n1.RGB, B=@n5)) {materialFunction=&MF_FX_EmissiveAdjustments}
  n6 = MaterialFunctionCall() {materialFunction=&LinearGradient}
  n7 = Multiply(
   A=@n6.VGradient
   B=Power(Base=OneMinus(None=@n6.VGradient), Exp=ScalarParameter() {defaultValue=3, parameterName=GradientPower}) {constExponent=3}
  )
  n8 = Multiply(A=@n1.A, B=Multiply(A=@n5, B=Multiply(A=@n7) {constB=2}))
  ShadingPathSwitch(Default=DepthFade(Opacity=@n8), Mobile=Multiply(A=@n8, B=ScalarParameter() {defaultValue=0.1, parameterName=Mobile Opacity}))
 ]
}
